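# Material function that adds a blinking red light to the MoosePorch backpack, layered over incoming material attributes.
# It overrides normal, base color, metallic, specular, roughness and emissive, with masks from the backpack and red-light textures.
# A time-driven sine creates the blink, while parameters set min and max light colors and brightness.
# Fresnel terms shape edge and glow, a bump-offset mask sample fakes depth, and emissive passes through MF_FX_EmissiveAdjustments.
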MaterialFunction /BRCosmetics/Accessories/FORT_Backpacks/BackPack_M_MED_MoosePorch/Materials/MF_MossPorch_lightBlinkingFunction
graph {nodes [
  n1 = NamedRerouteUsage()
  n2 = Multiply(A=NamedRerouteUsage(), B=NamedRerouteUsage())
  n3 = NamedRerouteUsage()
  n4 = Constant() {r=0.3}
  n5 = Constant() {r=0.02}
  n6 = NamedRerouteUsage()
  n7 = GetMaterialAttributes(None=FunctionInput() {inputType=FunctionInput_MaterialAttributes}) {attributeGetTypes [
    #1
    #2
    #3
    #4
    #5
   ]}
  n8 = SetMaterialAttributes!(
   MaterialAttributes=@n7.MaterialAttributes
   Normal=LinearInterpolate(A=NamedRerouteUsage(), B=NamedRerouteUsage(), Alpha=@n1)
   BaseColor=LinearInterpolate(A=NamedRerouteUsage(), B=NamedRerouteUsage(), Alpha=@n1)
   Metallic=LinearInterpolate(A=NamedRerouteUsage(), B=Constant() {r=0.6}, Alpha=@n1)
   Specular=Constant() {r=0.6}
   EmissiveColor=MaterialFunctionCall(Emissive=LinearInterpolate(B=@n2, Alpha=@n3)) {materialFunction=&MF_FX_EmissiveAdjustments}.Result
   Roughness=LinearInterpolate(A=NamedRerouteUsage(), B=LinearInterpolate(A=@n4, B=@n5, Alpha=@n6) {constA=0.2}, Alpha=NamedRerouteUsage())
   Anisotropy
  ) {attributeSetTypes [
    #1
    #2
    #3
    #4
    #6
    #5
    #7
   ]}
  FunctionOutput(None=@n8)
  NamedRerouteDeclaration(None=@n7.Normal) {name=Normal_input, nodeColor=<rgba 0 0.6719 1 1>}
  NamedRerouteDeclaration(None=@n7.BaseColor) {name=Color_input, nodeColor=<rgba 0 1 0.2734 1>}
  NamedRerouteDeclaration(None=@n7.Metallic) {name=MetallicInputt, nodeColor=<rgba 1 0 0.0469 1>}
  NamedRerouteDeclaration(None=@n7.Roughness) {name=RoughnessInput, nodeColor=<rgba 0.6875 1 0 1>}
  n9 = Constant() {r=0.8}
  n10 = Constant() {r=1}
  n11 = NamedRerouteUsage()
  n12 = Constant() {r=1}
  n13 = Constant() {r=1.8}
  n14 = Power(Base=Multiply(A=MaterialFunctionCall(Brightness=@n12, Exponent=@n13) {materialFunction=&MF_Fresnel_FX}.Fresnel) {constB=10}) {constExponent=20}
  NamedRerouteDeclaration(None=LinearInterpolate(A=LinearInterpolate(A=@n9, B=@n10, Alpha=@n11) {constB=0.8}, Alpha=Saturate(None=@n14))) {name=Edge, nodeColor=<rgba 1 0 0.9609 1>}
  n15 = Constant() {r=-0.6}
  n16 = ComponentMask(None=TextureCoordinate()) {g=true}
  n17 = Constant()
  n18 = LinearInterpolate(
   A=TextureSample(UVs=BumpOffset(Height=@n15)) {texture=/BRCosmetics/Accessories/FORT_Backpacks/BackPack_M_MED_MoosePorch/Textures/T_Redlight_M}.R
   B=TextureSample() {texture=&T_Redlight_M}.R
   Alpha=If(A=OneMinus(None=@n16), "A > B"=@n17, "A == B"=@n17, "A < B"=Constant() {r=1}) {constB=0.3}
  )
  n19 = LinearInterpolate(Alpha=NamedRerouteUsage()) {constA=2, constB=4}
  NamedRerouteDeclaration(None=Saturate(None=Multiply(A=@n18, B=@n19))) {name=Emissive_Tex, nodeColor=<rgba 0.4531 0 1 1>}
  n20 = TextureSample() {texture=&T_Redlight_M}
  NamedRerouteDeclaration(None=@n20.G) {name=OCC, nodeColor=<rgba 0 0.6484 1 1>}
  NamedRerouteDeclaration(None=@n20.B) {name=Dirt, nodeColor=<rgba 1 0 0.5625 1>}
  NamedRerouteDeclaration(None=MaterialFunctionCall!(
    Thickness=ScalarParameter() {defaultValue=1, parameterName=BackPack_Thickness}
    NormalWS=VertexNormalWS()
    PowerExponent=ScalarParameter() {defaultValue=4, parameterName=BackPack_PowerExponet}
   ) {materialFunction=/Game/Weapons/Prototype/ObjectMover/Materials/MF_NormalizedFresnel}.Power) {name=NormalizeFresnel, nodeColor=<rgba 1 0 0.8438 1>}
  n21 = Multiply(A=Time(), B=ScalarParameter() {defaultValue=0.5, group=MoosePorch, parameterName=BackPack_Blink_duration})
  n22 = MaterialFunctionCall(Value=@n21) {materialFunction=&LinearSine}
  n23 = Abs(None=Sine(None=@n21))
  n24 = Max(A=Subtract(A=Multiply(A=@n23) {constB=2}), B=Constant())
  n25 = ScalarParameter() {defaultValue=45, group=MoosePorch, parameterName=BackPack_Rate of change}
  n26 = Saturate(None=OneMinus(None=Power(Base=@n24, Exp=@n25) {constExponent=3}))
  NamedRerouteDeclaration(None=Multiply(A=Multiply(A=@n22.Direction, B=@n26)) {constB=0.4}) {name=Blinking, nodeColor=<rgba 1 0.1875 0 1>}
  n27 = TextureSample() {texture=/BRCosmetics/Accessories/FORT_Backpacks/BackPack_M_MED_MoosePorch/Textures/T_RedLight_N}
  n28 = ScalarParameter() {defaultValue=0.25, group=MoosePorch, parameterName=BackPack_NormalFlatness}
  NamedRerouteDeclaration(None=MaterialFunctionCall(
    BaseNormal=NamedRerouteUsage()
    AdditionalNormal=MaterialFunctionCall(Normal=@n27.RGB, Flatness=@n28) {materialFunction=&FlattenNormal}.Result
   ) {materialFunction=&BlendAngleCorrectedNormals}.Result) {name=NormalOutput, nodeColor=<rgba 0.9922 1 0 1>}
  n29 = LinearInterpolate(
   A=Constant3Vector() {constant=<rgba 0.03 0.0153 0.0126 1>}
   B=Constant3Vector() {constant=<rgba 0.035 0.0178 0.0147 1>}
   Alpha=NamedRerouteUsage()
  )
  n30 = VectorParameter() {parameterName=BackPack_RedLight_BaseColor, defaultValue=<rgba 0.0712 0.0015 0.0015 1>}
  n31 = Multiply(A=@n30.R) {constB=2}
  n32 = NamedRerouteUsage()
  n33 = Multiply(A=LinearInterpolate(B=LinearInterpolate(A=@n30.R, B=@n31, Alpha=@n32), Alpha=NamedRerouteUsage()), B=Saturate(None=NamedRerouteUsage()))
  n34 = NamedRerouteUsage()
  NamedRerouteDeclaration(None=Multiply(A=NamedRerouteUsage(), B=LinearInterpolate(A=@n29, B=@n33, Alpha=@n34))) {name=Color_output, nodeColor=<rgba 1 0 0.1172 1>}
  n35 = NamedRerouteUsage()
  n36 = Multiply(A=@n18, B=LinearInterpolate(Alpha=@n35) {constA=1, constB=4}) {constB=2}
  n37 = LinearInterpolate(Alpha=@n35) {constA=1, constB=4}
  NamedRerouteDeclaration(None=Saturate(None=Power(Base=@n36, Exp=@n37) {constExponent=1})) {name=EmissiveMask, nodeColor=<rgba 0.3594 1 0 1>}
  n38 = LinearInterpolate(A=Constant() {r=0.5}, Alpha=NamedRerouteUsage()) {constA=0.85}
  n39 = VectorParameter() {group=MoosePorch, parameterName=BackPack_Red_lightEmissiveColor, defaultValue=<rgba 0.975 0.0244 0.0244 1>}
  n40 = LinearInterpolate(
   A=Multiply(A=@n39.RGB, B=ScalarParameter() {defaultValue=0.8, group=MoosePorch, parameterName=BackPack_Light_min})
   B=Multiply(
    A=VectorParameter() {parameterName=BackPack_Red_lightEmissiveColor2, defaultValue=<rgba 0.975 0.0244 0.0244 1>}.RGB
    B=ScalarParameter() {defaultValue=5, group=MoosePorch, parameterName=BackPack_Light_max}
   )
   Alpha=NamedRerouteUsage()
  )
  n41 = NamedRerouteUsage()
  n42 = NamedRerouteUsage()
  n43 = Constant3Vector() {constant=<rgba 0.35 0 0.09 1>}
  n44 = NamedRerouteUsage()
  n45 = LinearInterpolate(A=Multiply(A=@n40, B=Multiply(A=@n41, B=@n40)), B=Add(A=Multiply(A=@n42, B=@n43), B=Multiply(A=@n42, B=@n40)), Alpha=@n44) {constA=1}
  n46 = ScalarParameter() {defaultValue=1.1, parameterName=BackPack_FresnelBrightness}
  n47 = ScalarParameter() {defaultValue=4, parameterName=BackPack_FresnelExponent}
  n48 = Multiply(
   A=LinearInterpolate(A=Multiply(A=@n45) {constB=1.2}, B=Multiply(A=@n45) {constB=1.5}, Alpha=@n44)
   B=LinearInterpolate(
    A=ScalarParameter() {defaultValue=0.2, parameterName=BackPack_FresnelLow}
    Alpha=MaterialFunctionCall(Brightness=@n46, Exponent=@n47) {materialFunction=&MF_Fresnel_FX}.Fresnel
   )
  )
  NamedRerouteDeclaration(None=Multiply(A=NamedRerouteUsage(), B=Multiply(A=@n38, B=@n48))) {name=EmissiveColor_output, nodeColor=<rgba 0 1 0.0391 1>}
  n49 = TextureSample() {texture=/BRCosmetics/Accessories/FORT_Backpacks/BackPack_M_MED_MoosePorch/Textures/T_Backpack_MoosePorch_E}
  NamedRerouteDeclaration(None=Saturate(None=Multiply(A=@n49.R) {constB=3})) {name=Mask, nodeColor=<rgba 1 0 0.8437 1>}
 ]}
