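# Parent material for the MoosePorch character cosmetic, with static switches for backpack and glider light-blinking variants.
# Character PreFX output takes an optional masked camo pattern in base color, plus emissive blended between flat and bump-offset parallax samples by a blue mask channel.
# Metallic-weighted roughness adjustment feeds a car clear coat function with its own coat normal.
# PostFX follows, with the PreFX normal restored before the blink switches.
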Material /BRCosmetics/Characters/Player/Male/Medium/Bodies/M_MED_MoosePorch_Math/Materials/M_MoosePorch_Parent
graph {
 properties {
  bUseMaterialAttributes=true
  bUsedWithMeshDeformer=true
  bUsedWithMorphTargets=true
  bUsedWithSkeletalMesh=true
  pixelDepthOffsetMode=PDOM_Legacy
  shadingModel=MSM_FromMaterialExpression
  usedShadingModels=Default Lit | Subsurface | Clear Coat | Subsurface Profile | Cloth
 }
 nodes [
  n1 = CameraVectorWS()
  n2 = ScalarParameter() {group=MoosePorch, parameterName=MoosePorch Light Offset}
  n3 = DotProduct(A=@n1, B=PixelNormalWS())
  Add(
   A=WorldPosition() {worldPositionShaderOffset=WPT_ExcludeAllShaderOffsets}.XY
   B=Multiply(A=ComponentMask(None=@n1) {g=true, r=true}, B=Divide(A=@n2, B=@n3))
  )
  n4 = MaterialFunctionCall() {materialFunction=&MF_CharacterParent_2023_PreFX}
  n5 = TextureSampleParameter2D() {
   constCoordinate=1
   group=MoosePorch
   parameterName=CamoTexture
   texture=/BRCosmetics/Characters/Player/Male/Medium/Bodies/M_MED_MoosePorch_Math/Skins/Brush/Textures/T_MoosePorch_Brush_Pattern_Body_D
  }
  n6 = LinearInterpolate(
   A=GetMaterialAttributes(None=@n4.Result) {attributeGetTypes [#1]}.Base Color
   B=@n5.RGB
   Alpha=ChannelMaskParameter(None=NamedRerouteUsage()) {group=MoosePorch, maskChannel=Green, parameterName=CamoMaskChannel, defaultValue=<rgba 0 1 0 0>}.R
  )
  n7 = GetMaterialAttributes(None=@n4.Result) {attributeGetTypes [#1]}
  n8 = GetMaterialAttributes(None=SetMaterialAttributes!(MaterialAttributes=@n4.Result, Base Color=StaticSwitchParameter(True=@n6, False=@n7.Base Color) {parameterName=bMoosePorchBrush Pattern}) {attributeSetTypes [#1]}) {attributeGetTypes [#2, #3]}
  n9 = TextureSampleParameter2D() {
   group=MoosePorch
   parameterName=MoosePorch Emissive
   samplerSource=SSM_Wrap_WorldGroupSettings
   texture=&T_BlacksRGB
  }
  n10 = ScalarParameter() {defaultValue=1, group=MoosePorch, parameterName=Emissive Brightness}
  n11 = TextureSampleParameter2D(UVs=BumpOffset(Height=Multiply(B=@n2) {constA=1}) {heightRatio=0.02}) {
   group=MoosePorch
   parameterName=MoosePorch Emissive
   samplerSource=SSM_Wrap_WorldGroupSettings
   texture=&T_BlacksRGB
  }
  n12 = NamedRerouteUsage()
  n13 = MaterialFunctionCall(Emissive=LinearInterpolate(A=Multiply(A=@n9.RGB, B=@n10), B=Multiply(A=@n11.RGB, B=@n10), Alpha=ComponentMask(None=@n12) {b=true})) {materialFunction=&MF_FX_EmissiveAdjustments}
  n14 = Multiply(A=@n8.Roughness, B=ScalarParameter() {defaultValue=1, group=MoosePorch, parameterName=Metal Roughness Multiplier})
  n15 = Saturate(None=Power(Base=@n8.Metallic))
  n16 = MaterialFunctionCall!(
   Mask=@n4.Tech Art Mask
   Fresnel=@n4.Tech Art Mask
   In=SetMaterialAttributes!(
    MaterialAttributes=SetMaterialAttributes!(MaterialAttributes=@n8.MaterialAttributes, Emissive Color=@n13.Result) {attributeSetTypes [#4]}
    Roughness=LinearInterpolate(A=@n8.Roughness, B=@n14, Alpha=@n15)
   ) {attributeSetTypes [#2]}
  ) {materialFunction=/Game/Global/MaterialFunctions/TechArt/MF_Advanced_CarClearCoat}
  ClearCoatNormalCustomOutput!(None=MaterialFunctionCall(BaseNormal=@n16.NormalWithOrangePeel, AdditionalNormal=@n16.NormalWithOrangePeel) {materialFunction=&BlendAngleCorrectedNormals}.Result)
  Constant3Vector() {constant=<rgba 0 0 1 0>}
  Constant() {r=2.5}
  NamedRerouteDeclaration(None=@n4.Tech Art Mask) {name=TechArtMask, nodeColor=<rgba 0.1953 1 0 1>}
  n17 = SetMaterialAttributes!(
   MaterialAttributes=MaterialFunctionCall(Material=@n16.Result, M=@n4.M, Pre Skinned Local Position=@n4.M, Distance Blend=@n4.M, Fresnel=@n4.M) {materialFunction=&MF_CharacterParent_2023_PostFX}.Result
   Normal=GetMaterialAttributes(None=@n4.Result) {attributeGetTypes [#5]}.Normal
  ) {attributeSetTypes [#5]}
  n18 = StaticSwitchParameter(
   True=MaterialFunctionCall!(In=@n17) {materialFunction=/BRCosmetics/Accessories/FORT_Backpacks/BackPack_M_MED_MoosePorch/Materials/MF_MossPorch_lightBlinkingFunction}.Result
   False=@n17
  ) {group=MoosePorch, parameterName=MossPorch_backPack}
  StaticSwitchParameter(
   True=MaterialFunctionCall!(In=@n18) {materialFunction=/BRCosmetics/Accessories/FORT_Backpacks/BackPack_M_MED_MoosePorch/Materials/MF_MossPorch_lightBlinkingGlider}.Result
   False=@n18
  ) {group=MoosePorch, parameterName=MossPorch_Glider}
 ]
}
