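# Character-part AnimBlueprint for the MoosePorch_Math outfit body, adding procedural motion to the pose copied from its parent mesh.
# AnimGraph chains deform and minigun Control Rigs, then additively adjusts both knee pads using alphas both derived from the right leg.
# EventGraph drives rig alphas from targeting, scope, weapon, hang-gliding, backpack and custom emote.
# It sets helmet variant flags and hides a material element for the matching glider, long backblings or gadget flags.
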AnimBlueprint /BRCosmetics/Characters/Player/Male/Medium/Bodies/M_MED_MoosePorch_Math/Meshes/M_MED_MoosePorch_Math_AnimBP
target_skeleton=&/M_MED_MoosePorch_Math_Skeleton
graphs [
 AnimGraph {nodes [
   n1 = "Transform (Modify) Bone - Bone: base_knee_pad_l"(
    ComponentPose=Local To Component(LocalPose=Use cached pose 'ControlRig_Cache'())
    Alpha=Get Knee_Pad_Rot_L() -> (Knee_Pad_Rot_L:double)
    Translation="-4,1,0"
    Rotation="0,-55,0"
    Scale="1,1,1"
   ) {
    rotationMode=BMM_Additive
    rotationSpace=BCS_BoneSpace
    translationMode=BMM_Additive
    translationSpace=BCS_BoneSpace
    alphaBoolBlend {blendOption=Linear}
    boneToModify {boneName=base_knee_pad_l}
    rotation {yaw=-55}
    translation=<-4 1>
   }
   n2 = "Transform (Modify) Bone - Bone: base_knee_pad_r"(ComponentPose=@n1, Alpha=Get Knee_Pad_Rot_R() -> (Knee_Pad_Rot_R:double), Translation="4,1,0", Rotation="0,-55,0", Scale="1,1,1") {
    rotationMode=BMM_Additive
    rotationSpace=BCS_BoneSpace
    translationMode=BMM_Additive
    translationSpace=BCS_BoneSpace
    alphaBoolBlend {blendOption=Linear}
    boneToModify {boneName=base_knee_pad_r}
    rotation {yaw=-55}
    translation=<4 1>
   }
   Output Pose(Result=Component To Local(ComponentPose=@n2)) {layerGroup=None}
   n3 = Control Rig!(Source=Copy Pose From Mesh() {bUseAttachedParent=true}, Alpha, skydiving:bool=Get bIsSkyDiving(), playingemotes:bool=Get bIsPlayingEmote()) -> (Pose) {
    alphaBoolBlend {blendOption=Linear}
    controlRigAssetReference {blueprintRigClass=&/M_MED_MoosePorch_Math_DeformRig.M_MED_MoosePorch_Math_DeformRig_C}
   }
   n4 = Control Rig!(
    Source=@n3
    Alpha
    isTargetingAlpha:double=Get isTargetingAlpha() -> (isTargetingAlpha:double)
    isHangGlidingAlpha:double=Get HangGlidingAlpha() -> (HangGlidingAlpha:double)
    isGunOutAlpha:double=Get isGunOut() -> (isGunOut:double)
    isBackpackEquipped:double=Get isBackpackEquipped() -> (isBackpackEquipped:double)
    isScoped:double=Get isScoped() -> (isScoped:double)
   ) -> (Pose) {
    alphaBoolBlend {blendOption=Linear}
    controlRigAssetReference {blueprintRigClass=&/M_MED_MoosePorch_Math_MiniGun_CtrlRig.M_MED_MoosePorch_Math_MiniGun_CtrlRig_C}
   }
   Save cached pose 'ControlRig_Cache'(Pose=@n4) {cacheName=ControlRig_Cache}
  ]}
 EventGraph {nodes [
   n1 = Cast To FortPlayerPawn(Object=GetOwningActor())
   > n2 = GetVariantChannelActiveIndexForItemActor!(
    execute
    Item:Athena Cosmetic Item Definition Object Reference=/BRCosmetics/Athena/Items/Cosmetics/Characters/Character_MoosePorch_Math
    Actor:Actor Object Reference=@n1
    Channel:Gameplay Tag Structure=(TagName="Cosmetics.Variant.Channel.Parts")
    VariantList:Array of Gameplay Tag Structures=Make Array!("[0]"=(TagName="Cosmetics.Variant.Property.Stage1"), "[1]":Gameplay Tag Structure=(TagName="Cosmetics.Variant.Property.Stage3")) -> (Array)
    bErrorIfChannelMissing:bool=false
    bErrorIfTagNotInArray:bool=false
   ) -> (then, Index:int=0, bIndexIsValid:bool, bFoundChannel:bool)
   n3 = "Equal (Integer)"(A=@n2, B=1)
   n4 = "Equal (Integer)"(A=@n2)
   > n5 = Branch(Condition=AND Boolean(A=OR Boolean(A=@n4, B=@n3), B=@n2.bIndexIsValid))
   > Set No Helmet(No Helmet:bool=false) -> (Output_Get:bool)
   > Set Helmet(Helmet:bool=true) -> (Output_Get:bool)
   Set No Helmet(execute=@n5.else, No Helmet=true)
   > Set Helmet(Helmet=false)
   n6 = Event BlueprintUpdateAnimation()
   > n7 = Sequence!(execute) -> (then_0, then_1, then_2:Exec, then_3:Exec, then_4:Exec)
   n8 = GetAnimInstance(self=GetOwnerSkeletalMeshComponent())
   Is Valid(exec=@n7.then_1, InputObject=@n8)
   > Set isCustomEmotePlaying(isCustomEmotePlaying:bool=Montage_IsPlaying(self=@n8, Montage=/BRCosmetics/Animation/Game/MainPlayer/Emotes/MoosePorch/Emote_MoosePorch_CMM_M)) -> (Output_Get:bool)
   > n9 = Get Dynamic Material Instance(SkeletalMeshComponent=GetPartSkeletalMeshComponent(), ElementIndex=1)
   n10 = Get MoosePorchGlider() -> (MoosePorchGlider:bool)
   n11 = Get bIsParachuteOpened()
   n12 = Get HasHoverJets() -> (HasHoverJets:bool)
   n13 = Get HasMissileSwarm() -> (HasMissileSwarm:bool)
   n14 = Get HasTurret() -> (HasTurret:bool)
   > n15 = Branch(Condition=OR Boolean!(A=AND Boolean(A=@n10, B=@n11), B=Get isBackblingLong() -> (isBackblingLong:bool), C:bool=OR Boolean(A=@n12, B=@n14, C=@n13)) -> (ReturnValue))
   > SetScalarParameterValue(self=@n9, ParameterName=Hide Element 01, Value=1)
   SetScalarParameterValue(execute=@n15.else, self=@n9, ParameterName=Hide Element 01, Value=0)
   n16 = Get Instigator!(self:Actor Object Reference=GetOwningActor()) -> (Instigator:Pawn Object Reference)
   Is Valid(exec=@n7.then_3, InputObject=@n16)
   > n17 = Cast To FortPlayerPawn(Object=@n16)
   > Set As Fort Player Pawn(As Fort Player Pawn:Fort Player Pawn Object Reference=@n17) -> (Output_Get:Fort Player Pawn Object Reference)
   n18 = Gravity Override State Function DEPRECATED(
    execute=@n7.then_0
    CharacterPartAnimInstance=Self-Reference()
    GameTimeInSeconds=GetGameTimeInSeconds()
    GravityOverrideParams=Get Gravity Override Params()
   )
   > Set GravityOverride(GravityOverride=@n18)
   n19 = Get SkelMesh()
   > Set Knee_Pad_Rot_R(Knee_Pad_Rot_R:double=GetAlphaFromReferenceTransformDelta(SkeletalMeshComponent=@n19, BoneName=calf_r, BaseName=thigh_r, InputRange=(X=0,Y=120), TwistAxis="0,0,1")) -> (Output_Get:double)
   > Set Knee_Pad_Rot_L(Knee_Pad_Rot_L:double=GetAlphaFromReferenceTransformDelta(SkeletalMeshComponent=@n19, BoneName=calf_r, BaseName=thigh_r, InputRange=(X=0,Y=120), TwistAxis="0,0,1")) -> (Output_Get:double)
   n20 = Get bIsTargeting()
   n21 = Get isCustomEmotePlaying() -> (isCustomEmotePlaying:bool)
   n22 = FInterpTo(
    Current=Get isTargetingAlpha()
    Target="To Float (Boolean)"(InBool=OR Boolean!(A=@n20, B=@n21) -> (ReturnValue))
    DeltaTime=@n6.DeltaTimeX
    InterpSpeed=10
   )
   Set isTargetingAlpha(execute=@n7.then_2, isTargetingAlpha:double=@n22) -> (Output_Get:double)
   > Branch(Condition=Get isCustomEmotePlaying())
   > Set isTargetingAlpha(isTargetingAlpha=@n22)
   > Set isGunOut(isGunOut:double=1) -> (Output_Get:double)
   Event OnBeginTargeting() {is=title K2Node_Event}
   > n23 = Branch(Condition=Get bIsInFirstPerson!(self:Fort Player Anim Instance Object Reference=GetOwnerFortPlayerAnimInstance!(self:Custom Character Part Anim Instance Object Reference) -> (ReturnValue:Fort Player Anim Instance Object Reference)) -> (bIsInFirstPerson:bool))
   > Set isScoped(isScoped:double=1) -> (Output_Get:double)
   Set isScoped(execute=[@n29.then, @n23.else], isScoped=0)
   n24 = Event OnCharacterCustomizationLoaded()
   > Set SkelMesh(SkelMesh=@n24.CustomBaseMesh)
   n25 = CheckForBackbling()
   > Set isBackblingLong(isBackblingLong:bool=@n25.bIsBackblingLong) -> (Output_Get:bool)
   n26 = NOT Boolean(A=@n25.bIsBackblingACape)
   > Set isBackpackEquipped(isBackpackEquipped:double="To Float (Boolean)"(InBool=AND Boolean(A=@n25, B=@n26))) -> (Output_Get:double)
   > n27 = Cast To FortPlayerPawn(Object=GetOwningActor())
   > Is Valid(InputObject=Get MoosePorchGliderSkelMesh() -> (MoosePorchGliderSkelMesh:Skeletal Mesh Object Reference))
   > n28 = Set Current Glider(Current Glider:Athena Glider Item Definition Object Reference=GetItemDefinitionTypeFromLoadoutWithActor!(
     Actor:Actor Object Reference=@n27
     LoadoutSlot:Gameplay Tag Structure=(TagName="LoadoutSlot.Glider")
     CosmeticType:Athena Cosmetic Item Definition Class Reference=/Script/FortniteGame.AthenaGliderItemDefinition
    ) -> (OutCosmeticItemDef:Athena Glider Item Definition Object Reference)) -> (Output_Get:Athena Glider Item Definition Object Reference)
   > Set MoosePorchGlider(MoosePorchGlider:bool="Equal (Object)"(A=@n28, B=Get GliderItemDef() -> (GliderItemDef:Athena Glider Item Definition Object Reference)) {is=title K2Node_CallFunction}) -> (Output_Get:bool)
   n29 = Event OnEndTargeting() {is=title K2Node_Event}
   Event OnParachuteOpened()
   > Switch on EGliderType(Selection=Get PlayerGliderType() -> (PlayerGliderType=HangGlider))
   > Set HangGlidingAlpha(HangGlidingAlpha:double=1) -> (Output_Get:double)
   Event OnParachuteClosed()
   > Set HangGlidingAlpha(HangGlidingAlpha=0)
   Event OnWeaponChanged()
   > n30 = Cast To FortPlayerAnimInstance(Object=GetAnimInstance(self=GetOwnerSkeletalMeshComponent()))
   > Set isGunOut(isGunOut=1)
   > n31 = Switch on EFortWeaponCoreAnimation!(execute, Selection:EFortWeaponCoreAnimation Enum=Get WeaponCoreAnim(self=@n30)) -> (
    Melee:Exec
    Pistol:Exec
    Shotgun:Exec
    PaperBlueprint:Exec
    Rifle:Exec
    MeleeOneHand:Exec
    MachinePistol:Exec
    RocketLauncher:Exec
    GrenadeLauncher:Exec
    GoingCommando:Exec
    AssaultRifle:Exec
    TacticalShotgun:Exec
    SniperRifle:Exec
    TrapPlacement:Exec
    ShoulderLauncher:Exec
    AbilityDecoTool:Exec
    Crossbow:Exec
    C4:Exec
    RemoteControl:Exec
    DualWield:Exec
    AR_BullPup:Exec
    AR_ForwardGrip:Exec
    MedPackPaddles:Exec
    SMG_P90:Exec
    AR_DrumGun:Exec
    Consumable_Small:Exec
    Consumable_Large:Exec
    Balloon:Exec
    MountedTurret:Exec
    CreativeTool:Exec
    ExplosiveBow:Exec
    AshtonIndigo:Exec
    AshtonChicago:Exec
    MeleeDualWield:Exec
    Unarmed:Exec
   )
   Set isGunOut(execute=[@n31.Melee, @n31.MeleeDualWield, @n31.MeleeOneHand, @n31.Unarmed], isGunOut=0)
   float * float(A=@n22, B=0.9) {is=title K2Node_PromotableOperator}
  ]}
]
variables [
 As Fort Player Pawn {type=FortPlayerPawn}
 Current Glider {type=AthenaGliderItemDefinition, default=null}
 GliderItemDef {type=AthenaGliderItemDefinition, default=/BRCosmetics/Athena/Items/Cosmetics/Gliders/Glider_MoosePorch}
 Gravity Override Params {type=GravityOverrideParamsStruct}
 GravityOverride {type=Vector}
 HangGlidingAlpha {type=number}
 HasHoverJets {type=boolean}
 HasMissileSwarm {type=boolean}
 HasTurret {type=boolean}
 Helmet {type=boolean}
 isBackblingLong {type=boolean}
 isBackpackEquipped {type=number}
 isCustomEmotePlaying {type=boolean}
 isGunOut {type=number}
 isScoped {type=number}
 isTargetingAlpha {type=number}
 Knee_Pad_Rot_L {type=number}
 Knee_Pad_Rot_R {type=number}
 MoosePorchGlider {type=boolean}
 MoosePorchGliderSkelMesh {type=SkeletalMesh, default=/BRCosmetics/Gadgets/Assets/VinderTech_GliderChute/Glider_MoosePorch/Meshes/MoosePorch_Glider}
 No Helmet {type=boolean}
 SkelMesh {type=SkeletalMeshComponent}
 SpineRotation {type=number}
]
